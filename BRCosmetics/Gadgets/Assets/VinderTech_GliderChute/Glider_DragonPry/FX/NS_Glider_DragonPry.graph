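# Niagara ribbon trail effect for the DragonPry glider, with separate deploy, sidewing, dark and colored trails.
# Each of the four emitters draws two anim trails between bone or socket pairs, pushed by added velocity that changes between gameplay and front end.
# User.Variant picks colors for all but DeployTrail, whose width ramps up with age.
# Those three also scale gameplay alpha by a system integrator, while front-end trails fade in over emitter age.
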NiagaraSystem /BRCosmetics/Gadgets/Assets/VinderTech_GliderChute/Glider_DragonPry/FX/NS_Glider_DragonPry
dependencies {
 usedDataInterfaces [&NiagaraDataInterfaceCurve, &NiagaraDataInterfaceColorCurve, &NiagaraDataInterfaceSocketReader]
 usedDynamicInputs [
  &Boolean_NotOperation
  &Boolean_AndOperation
  &SetBoolByFloatComparison
  &Abs_Float
  &AddVectorToPosition
  &TransformVector
  &ColorFromCurve
 ]
 usedModules [
  &self:SystemSpawnScript.SystemScriptSource.SystemScriptGraph.NiagaraNodeAssignment_0.SetVariables_#1
  &self:SystemSpawnScript.SystemScriptSource.SystemScriptGraph.NiagaraNodeAssignment_5.SetVariables_#2
  &SystemState
  &FN_Integrator
  &Timeline
  &self:SystemSpawnScript.SystemScriptSource.SystemScriptGraph.NiagaraNodeAssignment_1.SetVariables_#3
  &self:SystemSpawnScript.SystemScriptSource.SystemScriptGraph.NiagaraNodeAssignment_4.SetVariables_#4
  &self:SystemSpawnScript.SystemScriptSource.SystemScriptGraph.NiagaraNodeAssignment_8.SetVariables_#5
 ]
}
emitters {
 ColorTrail {
  emitterSpawnScript [&FN_RibbonIDOffsetUtility()]
  emitterUpdateScript [
   &EmitterState(Inactive Response=::NewEnumerator1, Life Cycle Mode=::NewEnumerator1, Loop Delay=1, Loop Duration=1, UseLoopDelay=true)
   &FN_SpawnAnimTrail(Number of Anim Trails=2)
  ]
  particleSpawnScript [
   &FN_InitializeAnimTrailBoneIndecies(First Bone or Socket Index=1, Second Bone or Socket Index=2)
   &FN_InitializeAnimTrailBoneIndecies(AnimTrail Selection=1, First Bone or Socket Index=5, Second Bone or Socket Index=6) {module=FN_InitializeAnimTrailBoneIndecies001}
   &FN_InitializeAnimTrail(
    AnimTrailWidth=1.2
    Color=&MultiplyLinearColor(
     Linear Color=&MergeHelper_ColorSwitch(
      0=<rgba 0.7656 0.1481 0.1156 5>
      1=<rgba 0 0 0 5>
      2=<rgba 0.1042 0.6985 1 5>
      3=&MakeCustomLinearColorFromBool(
       Bool=&Boolean_OrOperation(Bool1=$User.bIsFrontEnd, Bool2=$User.bIsFrontEndPreview)
       True Color=<rgba 0.8542 0 0.033 3>
       False Color=<rgba 0.2917 0.0273 0.1223 8>
      )
      Index=&MakeIntFromFloat(Float=$User.Variant, Conversion Mode=::NewEnumerator1)
      Num=::NewEnumerator2
     )
     Scale Factor=&MakeVector4_FromLinearColor(LinearColor=$System.TOD)
    )
    Lifetime=0.1
   )
   &AddVelocity(Velocity=&FN_FrontEndVector3Modifier(Gameplay=<0 2000 0>, Front End=<0 -2000 0>))
   &DynamicMaterialParameters(
    Index 0 Param 1=&RandomRangeFloat(
     Minimum=0
     Maximum=100
     Randomness Mode=::NewEnumerator1
     Random Seed=&ReturnParticleIDAttributes(Particle ID Attribute=::NewEnumerator1, Niagara ID=$Particles.RibbonID)
     Fixed Random Seed=true
    )
    Index 0 Param 2=&RandomRangeFloat(
     Minimum=0
     Maximum=100
     Randomness Mode=::NewEnumerator1
     Random Seed=&ReturnParticleIDAttributes(Particle ID Attribute=::NewEnumerator1, Niagara ID=$Particles.RibbonID)
     Fixed Random Seed=true
    )
    Module.Index0Float0Write=true
    Module.Index0Float1Write=true
    Module.Index0Float2Write=true
    Module.Index0Float3Write=true
   )
  ]
  particleUpdateScript [
   &InheritSourceMovement(
    Applied Position Scale=&Multiply_VectorByFloat(
     Vector=<1 1 1>
     Float=&FloatFromCurve(FloatCurve=DataInterface() {propertyValues {Curve {keys [
          {interpMode=RCIM_Cubic, tangentMode=RCTM_User, time=0.75, value=1}
          {arriveTangent=-0.3795, interpMode=RCIM_Cubic, leaveTangent=-0.3795, tangentMode=RCTM_User, time=1, value=0.9}
         ]}}})
    )
    ApplyForce=false
    ApplyPosition=true
    ApplyPositionFalloff=false
   )
   &SolveForcesAndVelocity(Clamp Velocity=false, Limit Acceleration=false)
   &FN_UpdateAnimTrail() {module=FN_UpdateAnimTrail001}
   &ScaleColor(
    Scale Alpha=&Multiply_Float(
     A=&FloatFromCurve(
      FloatCurve=DataInterface() {propertyValues {Curve {keys [
          {arriveTangent=-1.5186, interpMode=RCIM_Cubic, leaveTangent=-1.5186, tangentMode=RCTM_User, time=0, value=0.8}
          {arriveTangent=-0.6904, interpMode=RCIM_Cubic, leaveTangent=-0.6904, tangentMode=RCTM_User, time=0.3, value=0.25}
          {interpMode=RCIM_Cubic, tangentMode=RCTM_User, time=1, value=0}
         ]}}}
      Scale Curve=0.55
     )
     B=&FN_FrontEndFloatModifier(
      Gameplay=1
      Front End=&FloatFromCurve(FloatCurve=DataInterface() {propertyValues {Curve {keys [<at 1 0>,<at 3 1>]}}}, CurveIndex=$Emitter.Age)
     )
    )
    ScaleA=true
    ScaleRGB=true
   ) {module=ScaleColor001}
   &ScaleColor(
    Color Value To Scale=$Particles.Color
    Scale Alpha=&Multiply_Float(
     A=&FN_FrontEndFloatModifier(Gameplay=&Multiply_Float(A=$System.FN_Integrator.Value, B=1), Front End=0.65)
     B=1
    )
    ScaleA=true
    ScaleRGB=true
   )
   &InterpolateOverTime(Initial Float Value:NiagaraFloat=$System.Timeline.Time, Target Float Value=1, Use Set Interpolation Rate=true)
  ]
  rendererClasses [&NiagaraRibbonRendererProperties]
  renderers [{rendererClass=&NiagaraRibbonRendererProperties}]
 }
 Dark_Trails {
  emitterSpawnScript [&FN_RibbonIDOffsetUtility()]
  emitterUpdateScript [
   &EmitterState(Inactive Response=::NewEnumerator1, Life Cycle Mode=::NewEnumerator1, Loop Delay=1, Loop Duration=1, UseLoopDelay=true)
   &FN_SpawnAnimTrail(Number of Anim Trails=2)
  ]
  particleSpawnScript [
   &FN_InitializeAnimTrailBoneIndecies(First Bone or Socket Index=6, Second Bone or Socket Index=7)
   &FN_InitializeAnimTrailBoneIndecies(AnimTrail Selection=1, First Bone or Socket Index=2, Second Bone or Socket Index=3) {module=FN_InitializeAnimTrailBoneIndecies001}
   &FN_InitializeAnimTrail(
    Color=&MergeHelper_ColorSwitch(
     0=<rgba 0.1146 0.1146 0.1146 5>
     1=<rgba 0.1146 0.1146 0.1146 5>
     2=<rgba 0.8958 0.8958 0.8958 1>
     3=<rgba 0.1354 0.024 0.1195 3>
     Index=&MakeIntFromFloat(Float=$User.Variant)
     Num=::NewEnumerator2
    )
    Lifetime=0.15
   )
   &AddVelocity(Velocity=&FN_FrontEndVector3Modifier(Gameplay=<0 2000 0>, Front End=<0 -2000 0>))
  ]
  particleUpdateScript [
   &InheritSourceMovement(
    Applied Position Scale=&Multiply_VectorByFloat(
     Vector=<1 1 1>
     Float=&FloatFromCurve(FloatCurve=DataInterface() {propertyValues {Curve {keys [
          {interpMode=RCIM_Cubic, tangentMode=RCTM_User, time=0.75, value=1}
          {arriveTangent=-0.3795, interpMode=RCIM_Cubic, leaveTangent=-0.3795, tangentMode=RCTM_User, time=1, value=0.9}
         ]}}})
    )
    ApplyForce=false
    ApplyPosition=true
    ApplyPositionFalloff=false
   )
   &SolveForcesAndVelocity(Clamp Velocity=false, Limit Acceleration=false)
   &FN_UpdateAnimTrail() {module=FN_UpdateAnimTrail001}
   &ScaleColor(
    Scale Alpha=&Multiply_Float(
     A=&FloatFromCurve(
      FloatCurve=DataInterface() {propertyValues {Curve {keys [
          {arriveTangent=-1.5186, interpMode=RCIM_Cubic, leaveTangent=-1.5186, tangentMode=RCTM_User, time=0, value=0.8}
          {arriveTangent=-0.6904, interpMode=RCIM_Cubic, leaveTangent=-0.6904, tangentMode=RCTM_User, time=0.3, value=0.25}
          {interpMode=RCIM_Cubic, tangentMode=RCTM_User, time=1, value=0}
         ]}}}
      Scale Curve=1.35
     )
     B=&FN_FrontEndFloatModifier(
      Gameplay=1
      Front End=&FloatFromCurve(FloatCurve=DataInterface() {propertyValues {Curve {keys [<at 1 0>,<at 3 1>]}}}, CurveIndex=$Emitter.Age)
     )
    )
    ScaleA=true
    ScaleRGB=true
   ) {module=ScaleColor001}
   &ScaleColor(
    Color Value To Scale=$Particles.Color
    Scale Alpha=&Multiply_Float(
     A=&FN_FrontEndFloatModifier(Gameplay=&Multiply_Float(A=$System.FN_Integrator.Value, B=1), Front End=0.65)
     B=1
    )
    ScaleA=true
    ScaleRGB=true
   )
   &InterpolateOverTime(Initial Float Value:NiagaraFloat=$System.Timeline.Time, Target Float Value=1, Use Set Interpolation Rate=true)
  ]
  rendererClasses [&NiagaraRibbonRendererProperties]
  renderers [{rendererClass=&NiagaraRibbonRendererProperties}]
 }
 DeployTrail {
  emitterUpdateScript [
   &EmitterState(Life Cycle Mode=::NewEnumerator1, Loop Behavior=::NewEnumerator1, Loop Delay=1, Loop Duration=0.5, UseLoopDelay=false)
   &FN_SpawnAnimTrail(Number of Anim Trails=2)
  ]
  particleSpawnScript [
   &FN_InitializeAnimTrailBoneIndecies()
   &FN_InitializeAnimTrailBoneIndecies(AnimTrail Selection=1, First Bone or Socket Index=4, Second Bone or Socket Index=5) {module=FN_InitializeAnimTrailBoneIndecies001}
   &FN_InitializeAnimTrail(Color=<rgba 0.9427 0.9427 0.9427 0.5>)
   &AddVelocity(Velocity=&FN_FrontEndVector3Modifier(Gameplay=<0 1000 0>, Front End=<-2000 0 0>))
  ]
  particleUpdateScript [
   &InheritSourceMovement(
    Applied Position Scale=&Multiply_VectorByFloat(
     Vector=<1 1 1>
     Float=&FloatFromCurve(FloatCurve=DataInterface() {propertyValues {Curve {keys [
          {interpMode=RCIM_Cubic, tangentMode=RCTM_User, time=0.66, value=1}
          {arriveTangent=-1.9042, interpMode=RCIM_Cubic, leaveTangent=-1.9042, tangentMode=RCTM_User, time=1, value=0.5}
         ]}}})
    )
    ApplyForce=false
    ApplyPosition=true
    ApplyPositionFalloff=false
   )
   &SolveForcesAndVelocity(Clamp Velocity=false, Limit Acceleration=false)
   &FN_UpdateAnimTrail() {module=FN_UpdateAnimTrail001}
   &ScaleColor(
    Scale Alpha=&Multiply_Float(
     A=&FloatFromCurve(
      FloatCurve=DataInterface() {propertyValues {Curve {keys [
          {arriveTangent=-1.5186, interpMode=RCIM_Cubic, leaveTangent=-1.5186, tangentMode=RCTM_User, time=0, value=0.8}
          {arriveTangent=-0.6904, interpMode=RCIM_Cubic, leaveTangent=-0.6904, tangentMode=RCTM_User, time=0.3, value=0.25}
          {interpMode=RCIM_Cubic, tangentMode=RCTM_User, time=1, value=0}
         ]}}}
      Scale Curve=1.5
     )
     B=&FN_FrontEndFloatModifier(
      Gameplay=1
      Front End=&FloatFromCurve(FloatCurve=DataInterface() {propertyValues {Curve {keys [<at 1 0>,<at 3 1>]}}}, CurveIndex=$Emitter.Age)
     )
    )
    ScaleA=true
    ScaleRGB=true
   ) {module=ScaleColor001}
   &ScaleColor(Color Value To Scale=$Particles.Color, ScaleA=true, ScaleRGB=true)
   &ScaleColor(
    Scale RGB=&MakeVectorFromLinearColorRGB(Color=&FNTimeOfDayColor(Night Color=<rgba 0.45 0.462 1.5 1>, Morning Color=<rgba 1 1 1 1>, Day Color=<rgba 1 1 1 1>, Evening Color=<rgba 1 1 1 1>))
    ScaleA=false
    ScaleRGB=true
   ) {module=ScaleColor002}
   &ScaleRibbonWidth(Ribbon Width Scale=&FloatFromCurve(FloatCurve=DataInterface() {propertyValues {Curve {keys [<at 0.1 0.5>,<at 0.2 1>,<at 1 1>]}}}, CurveIndex=$Emitter.Age))
  ]
  rendererClasses [&NiagaraRibbonRendererProperties]
  renderers [{rendererClass=&NiagaraRibbonRendererProperties}]
 }
 Sidewings {
  emitterSpawnScript [&FN_RibbonIDOffsetUtility()]
  emitterUpdateScript [
   &EmitterState(Inactive Response=::NewEnumerator1, Life Cycle Mode=::NewEnumerator1, Loop Delay=1, UseLoopDelay=true)
   &FN_SpawnAnimTrail(Number of Anim Trails=2)
  ]
  particleSpawnScript [
   &FN_InitializeAnimTrailBoneIndecies()
   &FN_InitializeAnimTrailBoneIndecies(AnimTrail Selection=1, First Bone or Socket Index=4, Second Bone or Socket Index=5) {module=FN_InitializeAnimTrailBoneIndecies001}
   &FN_InitializeAnimTrail(
    AnimTrailWidth=0.8
    Color=&MultiplyLinearColor(
     Linear Color=&MergeHelper_ColorSwitch(
      0=<rgba 0.5052 0.5052 0.5052 0.3>
      1=<rgba 0.5052 0.1868 0.1929 0.3>
      2=?
      3=?
      Index=&MakeIntFromFloat(Float=$User.Variant, Conversion Mode=::NewEnumerator1)
      Num=::NewEnumerator0
     )
     Scale Factor=&MakeVector4_FromLinearColor(LinearColor=$System.TOD)
    )
    Lifetime=0.2
   )
   &AddVelocity(Velocity=&FN_FrontEndVector3Modifier(Gameplay=<0 2500 0>, Front End=<0 -2500 0>))
  ]
  particleUpdateScript [
   &InheritSourceMovement(
    Applied Position Scale=&Multiply_VectorByFloat(
     Vector=<1 1 1>
     Float=&FloatFromCurve(FloatCurve=DataInterface() {propertyValues {Curve {keys [
          {interpMode=RCIM_Cubic, tangentMode=RCTM_User, time=0.66, value=1}
          {arriveTangent=-1.9042, interpMode=RCIM_Cubic, leaveTangent=-1.9042, tangentMode=RCTM_User, time=1, value=0.5}
         ]}}})
    )
    ApplyForce=false
    ApplyPosition=true
    ApplyPositionFalloff=false
   )
   &SolveForcesAndVelocity(Clamp Velocity=false, Limit Acceleration=false)
   &FN_UpdateAnimTrail() {module=FN_UpdateAnimTrail001}
   &ScaleColor(
    Scale Alpha=&Multiply_Float(
     A=&FloatFromCurve(
      FloatCurve=DataInterface() {propertyValues {Curve {keys [
          {arriveTangent=-1.5186, interpMode=RCIM_Cubic, leaveTangent=-1.5186, tangentMode=RCTM_User, time=0, value=0.8}
          {arriveTangent=-0.6904, interpMode=RCIM_Cubic, leaveTangent=-0.6904, tangentMode=RCTM_User, time=0.3, value=0.25}
          {interpMode=RCIM_Cubic, tangentMode=RCTM_User, time=1, value=0}
         ]}}}
      Scale Curve=2
     )
     B=&FN_FrontEndFloatModifier(
      Gameplay=1
      Front End=&FloatFromCurve(FloatCurve=DataInterface() {propertyValues {Curve {keys [<at 1 0>,<at 3 1>]}}}, CurveIndex=$Emitter.Age)
     )
    )
    ScaleA=true
    ScaleRGB=true
   ) {module=ScaleColor001}
   &ScaleColor(
    Color Value To Scale=$Particles.Color
    Scale Alpha=&Multiply_Float(
     A=&FN_FrontEndFloatModifier(Gameplay=&Multiply_Float(A=$System.FN_Integrator.Value, B=0.65), Front End=0.65)
     B=1
    )
    ScaleA=true
    ScaleRGB=true
   )
   &InterpolateOverTime(Initial Float Value:NiagaraFloat=$System.Timeline.Time, Target Float Value=1, Use Set Interpolation Rate=true)
  ]
  rendererClasses [&NiagaraRibbonRendererProperties]
  renderers [{rendererClass=&NiagaraRibbonRendererProperties}]
 }
}
summary {
 systemName=NS_Glider_DragonPry
 emitters [
  {emitterName=DeployTrail, rendererClasses [&NiagaraRibbonRendererProperties]}
  {emitterName=Sidewings, rendererClasses [&NiagaraRibbonRendererProperties]}
  {emitterName=Dark_Trails, rendererClasses [&NiagaraRibbonRendererProperties]}
  {emitterName=ColorTrail, rendererClasses [&NiagaraRibbonRendererProperties]}
 ]
 userVariables [
  User.bIsFrontEnd {type=NiagaraBool}
  User.bIsFrontEndPreview {type=NiagaraBool}
  User.bIsFullyDeployed {type=NiagaraBool}
  User.ForwardDot {type=NiagaraFloat, defaultValue=0, description=""}
  User.RibbonIDOffset {type=NiagaraInt32}
  User.RightDot {type=NiagaraFloat, defaultValue=0, description=""}
  User.Variant {type=NiagaraFloat, defaultValue=0, description=""}
 ]
}
