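# Niagara system for the DragonPry glider: timed smoke bursts, wing streaks, blade flares and deploy ribbon trails.
# Smoke and StreakySmoke follow Socket Reader sockets, using loop count and particle parity to fire from socket 0, then successive pairs.
# Delayed BladeFlares and DeployTrail ribbons use sockets 13 and 14, with colors selected by User.Variant.
# Smoke and streaks take a System.TOD tint, while front-end modifiers adjust velocities and trail fade.
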NiagaraSystem /BRCosmetics/Gadgets/Assets/VinderTech_GliderChute/Glider_DragonPry/FX/NS_Glider_DragonPry_Smoke
dependencies {
 usedDataInterfaces [
  &NiagaraDataInterfaceCurve
  &NiagaraDataInterfaceVector2DCurve
  &NiagaraDataInterfaceSpriteRendererInfo
  &NiagaraDataInterfaceColorCurve
  &NiagaraDataInterfaceVectorCurve
  &NiagaraDataInterfaceSocketReader
 ]
 usedDynamicInputs [&ConvertVectorToPosition]
 usedModules [
  &self:SystemSpawnScript.SystemScriptSource.SystemScriptGraph.NiagaraNodeAssignment_0.SetVariables_#1
  &self:SystemSpawnScript.SystemScriptSource.SystemScriptGraph.NiagaraNodeAssignment_1.SetVariables_#2
  &SystemState
 ]
}
emitters {
 BladeFlares {
  emitterUpdateScript [
   &EmitterState(Life Cycle Mode=::NewEnumerator1, Loop Behavior=::NewEnumerator1, Loop Delay=0.99, Loop Duration=1, UseLoopDelay=true)
   &SpawnBurst_Instantaneous(Spawn Count=2, Use Spawn Probability=false, UseLoopCountLimit=true)
  ]
  particleSpawnScript [
   &PartitionParticles()
   &InitializeParticle(
    Lifetime=0.3
    Sprite Size=&Multiply_Vector2D_ByFloat(Float=150, Vector2D=<1 1>)
    Sprite Size Mode=::NewEnumerator4
    UsePositionOffset=false
    Write Material Random=true
   )
   &CameraOffset(Camera Offset Amount=5)
   /SetVariables(Particles.CameraOffsetFlare=150) {isSetParameters=true}
  ]
  particleUpdateScript [
   &ParticleState()
   &SocketLocation(
    Interpolate Socket Sampling=true
    SamplePosOffset=true
    Sampled Position Offset=<5 0 0>
    Socket Index=&Add_Integer(A=$Particles.PartitionParticles.Partition, B=13)
    Socket Reader=$System.Socket Reader
    Socket Sampling Mode=::NewEnumerator3
    VelocitySpeedLimit=false
   ) {module=SocketLocation001}
   &ScaleSpriteSize(Uniform Curve Sprite Scale=DataInterface() {propertyValues {Curve {keys [{interpMode=RCIM_Cubic, tangentMode=RCTM_User, time=0, value=1}, {interpMode=RCIM_Cubic, tangentMode=RCTM_User, time=1, value=0.2}]}}})
   &SolveForcesAndVelocity(Clamp Velocity=false, Limit Acceleration=false)
   &ScaleColor(
    Scale Alpha=&FloatFromCurve(FloatCurve=DataInterface() {propertyValues {Curve {keys [
         <at 0 0 RCIM_Cubic>
         {arriveTangent=-0.5651, interpMode=RCIM_Cubic, leaveTangent=-0.5651, tangentMode=RCTM_User, time=0.1, value=1}
         {arriveTangent=-0.5113, interpMode=RCIM_Cubic, leaveTangent=-0.5113, tangentMode=RCTM_User, time=0.8, value=0}
        ]}}})
    Scale RGB=&MakeVectorFromLinearColorRGB(Color=&MergeHelper_ColorSwitch(
      0=&ColorFromCurve(ColorCurve=DataInterface() {propertyValues {
         AlphaCurve {keys [<at 0 1 RCIM_Cubic>]}
         BlueCurve {keys [<at 0 0 RCIM_Cubic>,<at 0.4755 0 RCIM_Cubic>]}
         GreenCurve {keys [<at 0 0.0206 RCIM_Cubic>,<at 0.4755 0.0033 RCIM_Cubic>]}
         RedCurve {keys [<at 0 1 RCIM_Cubic>,<at 0.4755 0.0885 RCIM_Cubic>]}
        }})
      1=&ColorFromCurve(ColorCurve=DataInterface() {propertyValues {
         AlphaCurve {keys [<at 0 1 RCIM_Cubic>]}
         BlueCurve {keys [<at 0 0 RCIM_Cubic>,<at 0.4755 0 RCIM_Cubic>]}
         GreenCurve {keys [<at 0 0.0206 RCIM_Cubic>,<at 0.4755 0.0033 RCIM_Cubic>]}
         RedCurve {keys [<at 0 1 RCIM_Cubic>,<at 0.4755 0.0885 RCIM_Cubic>]}
        }})
      2=&ColorFromCurve(ColorCurve=DataInterface() {propertyValues {
         AlphaCurve {keys [<at 0 1 RCIM_Cubic>]}
         BlueCurve {keys [<at 0 1 RCIM_Cubic>,<at 0.4755 0.0885 RCIM_Cubic>]}
         GreenCurve {keys [<at 0 0.9206 RCIM_Cubic>,<at 0.4755 0.0116 RCIM_Cubic>]}
         RedCurve {keys [<at 0 0.1615 RCIM_Cubic>,<at 0.4755 0.0068 RCIM_Cubic>]}
        }})
      3=&ColorFromCurve(ColorCurve=DataInterface() {propertyValues {
         AlphaCurve {keys [<at 0 1 RCIM_Cubic>]}
         BlueCurve {keys [<at 0 0.2211 RCIM_Cubic>,<at 0.4755 0.0885 RCIM_Cubic>]}
         GreenCurve {keys [<at 0 0 RCIM_Cubic>,<at 0.4755 0.0091 RCIM_Cubic>]}
         RedCurve {keys [<at 0 0.6875 RCIM_Cubic>,<at 0.4755 0.0676 RCIM_Cubic>]}
        }})
      Index=&MakeIntFromFloat(Float=$User.Variant)
      Num=::NewEnumerator2
     ))
    ScaleA=true
    ScaleRGB=true
   )
   /SetVariables(
    Particles.ColorGlow:LinearColor=&MultiplyLinearColor(Linear Color=$Particles.Color, Scale Factor=&MakeVector4_FromLinearColor(LinearColor=<rgba 3 3 3 1>))
    Particles.SpriteSizeGlow:Vector2f=&Multiply_Vector2D_ByFloat(Float=2, Vector2D=$Particles.SpriteSize)
   ) {isSetParameters=true}
  ]
  rendererClasses [&NiagaraSpriteRendererProperties]
  renderers [
   {rendererClass=&NiagaraSpriteRendererProperties}
   {rendererClass=&NiagaraSpriteRendererProperties, rendererIndex=1}
  ]
 }
 DeployTrail {
  emitterUpdateScript [
   &EmitterState(Life Cycle Mode=::NewEnumerator1, Loop Behavior=::NewEnumerator1, Loop Delay=1, Loop Duration=0.12, UseLoopDelay=true)
   &FN_SpawnAnimTrail(Number of Anim Trails=2)
  ]
  particleSpawnScript [
   &FN_InitializeAnimTrailBoneIndecies(First Bone or Socket Index=13, Second Bone or Socket Index=13)
   &FN_InitializeAnimTrailBoneIndecies(AnimTrail Selection=1, First Bone or Socket Index=14, Second Bone or Socket Index=14) {module=FN_InitializeAnimTrailBoneIndecies001}
   &FN_InitializeAnimTrail(
    AnimTrailWidth=50
    Color=&MergeHelper_ColorSwitch(
     0=<rgba 2 0.3021 0.3021 2>
     1=<rgba 0 0 0 2>
     2=<rgba 0 1.64 2 2>
     3=<rgba 2 0 0.7812 2>
     Index=&MakeIntFromFloat(Float=$User.Variant, Conversion Mode=::NewEnumerator1)
     Num=::NewEnumerator2
    )
    Lifetime=0.2
   )
   &AddVelocity(Velocity=&FN_FrontEndVector3Modifier(Gameplay=<0 0 0>, Front End=<-2000 0 0>))
  ]
  particleUpdateScript [
   &InheritSourceMovement(
    Applied Position Scale=&Multiply_VectorByFloat(
     Vector=<1 1 1>
     Float=&FloatFromCurve(FloatCurve=DataInterface() {propertyValues {Curve {keys [
          {interpMode=RCIM_Cubic, tangentMode=RCTM_User, time=0.66, value=1}
          {arriveTangent=-1.9042, interpMode=RCIM_Cubic, leaveTangent=-1.9042, tangentMode=RCTM_User, time=1, value=0.5}
         ]}}})
    )
    ApplyForce=false
    ApplyPosition=true
    ApplyPositionFalloff=false
   )
   &SolveForcesAndVelocity(Clamp Velocity=false, Limit Acceleration=false)
   &FN_UpdateAnimTrail() {module=FN_UpdateAnimTrail001}
   &ScaleColor(
    Scale Alpha=&Multiply_Float(
     A=&Multiply_Float(
      A=&FloatFromCurve(
       FloatCurve=DataInterface() {propertyValues {Curve {keys [
           {arriveTangent=-1.5186, interpMode=RCIM_Cubic, leaveTangent=-1.5186, tangentMode=RCTM_User, time=0, value=0.8}
           {arriveTangent=-0.6904, interpMode=RCIM_Cubic, leaveTangent=-0.6904, tangentMode=RCTM_User, time=0.3, value=0.25}
           {interpMode=RCIM_Cubic, tangentMode=RCTM_User, time=1, value=0}
          ]}}}
       Scale Curve=1.5
      )
      B=&FN_FrontEndFloatModifier(
       Gameplay=1
       Front End=&FloatFromCurve(FloatCurve=DataInterface() {propertyValues {Curve {keys [<at 1 0>,<at 3 1>]}}}, CurveIndex=$Emitter.Age)
      )
     )
     B=&FloatFromCurve(
      FloatCurve=DataInterface() {propertyValues {Curve {keys [<at 0 0>, <at 0.1 1>, {arriveTangent=-8.3333, time=0.4, value=1}, <at 0.5 0>]}}}
      CurveIndex=&Add_Float(A=$Emitter.Age, B=-1)
     )
    )
    ScaleA=true
    ScaleRGB=true
   ) {module=ScaleColor001}
   &ScaleRibbonWidth(Ribbon Width Scale=&FloatFromCurve(FloatCurve=DataInterface() {propertyValues {Curve {keys [<at 0.1 0.1>,<at 0.5 1>,<at 1 1>]}}}))
  ]
  rendererClasses [&NiagaraRibbonRendererProperties]
  renderers [{rendererClass=&NiagaraRibbonRendererProperties}]
 }
 Smoke {
  emitterUpdateScript [
   &EmitterState(
    Delay First Loop Only=true
    Life Cycle Mode=::NewEnumerator1
    Loop Behavior=::NewEnumerator2
    Loop Count=7
    Loop Delay=1.6
    Loop Duration=0.1
    UseLoopDelay=true
   )
   &SpawnBurst_Instantaneous(Spawn Count=6, Use Spawn Probability=false, UseLoopCountLimit=false)
  ]
  particleSpawnScript [
   &InitializeParticle(
    Color=&MultiplyLinearColor(Linear Color=<rgba 1 1 1 0.5>, Scale Factor=&MakeVector4_FromLinearColor(LinearColor=$System.TOD))
    Lifetime=&RandomRangeFloat(Minimum=0.3, Maximum=0.5)
    Position Mode=::NewEnumerator0
    Sprite Rotation Angle=&RandomRangeFloat(Minimum=0, Maximum=1)
    Sprite Rotation Mode=::NewEnumerator1
    Sprite Size=&Vector2DFromFloat(Value=&RandomRangeFloat(Minimum=120, Maximum=150))
    Sprite Size Mode=::NewEnumerator4
    Write Material Random=true
   )
   &AddVelocity(
    Cone Angle=&RandomRangeFloat(Minimum=5, Maximum=20)
    Cone Axis=&Multiply_Vector(
     A=&MakeCustomVectorFromBool(
      Bool=&SetBoolByIntComparison(A=$Emitter.LoopCount, B=0, Comparison Type=::NewEnumerator4)
      True Vector=&RandomRangeVector(Minimum=<-0.3 0.2 0.5>, Maximum=<0.3 -0.2 1>)
      False Vector=&RandomRangeVector(Minimum=<-0.3 1 0.3>, Maximum=<0.03 1 0.5>)
     )
     B=&FN_FrontEndVector3Modifier(Gameplay=<1 1 1>, Front End=<-1 -1 1>)
    )
    Distribution Along Cone Axis=0
    Module.UseConeFalloff=false
    Module.UseVelDistribution=false
    Override Seed=false
    Velocity Mode=::NewEnumerator2
    Velocity Speed=&RandomRangeFloat(Minimum=100, Maximum=150)
   ) {module=AddVelocity001}
   /SetVariables(Particles.CustomID:NiagaraInt32=&MakeCustomIntFromBool(
     Bool=&SetBoolByIntComparison(A=$Emitter.LoopCount, B=0, Comparison Type=::NewEnumerator4)
     True INT=0
     False INT=/Niagara/DynamicInputs/Subtract/Subtract_Integer(
      A:NiagaraInt32=&Multiply_Int(A=$Emitter.LoopCount, B=2)
      B:NiagaraInt32=&ModuloInt(A=&ReturnExecIndex(), B=2)
     )
    )) {isSetParameters=true}
  ]
  particleUpdateScript [
   &ParticleState()
   &Drag(Use Linear Drag=true, Use Rotational Drag=true)
   &SocketLocation(
    Interpolate Socket Sampling=true
    Orientation Sampling=::NewEnumerator1
    Position Sampling=::NewEnumerator1
    SamplePosOffset=true
    Sampled Position Offset=&MakeCustomVectorFromBool(
     Bool=&SetBoolByIntComparison(A=$Emitter.LoopCount, B=0, Comparison Type=::NewEnumerator4)
     True Vector=<0 0 0>
     False Vector=&RandomRangeVector(Minimum=<-10 0 -5>, Maximum=<10 0 -5>)
    )
    Sampled Position Offset Coordinate Space:ENiagaraSkelMeshTransforms=::NewEnumerator2
    Socket Index=$Particles.Initial.CustomID
    Socket Reader=$System.Socket Reader
    Socket Sampling Mode=::NewEnumerator3
    VelocitySpeedLimit=false
   )
   &SolveForcesAndVelocity(Clamp Velocity=false, Limit Acceleration=false)
   &ScaleSpriteSize(Uniform Curve Sprite Scale=DataInterface() {propertyValues {Curve {keys [
        {arriveTangent=2.194, interpMode=RCIM_Cubic, time=0, value=0}
        {arriveTangent=0.2449, interpMode=RCIM_Cubic, leaveTangent=0.2449, tangentMode=RCTM_User, time=1, value=1}
       ]}}})
   /Niagara/Modules/Update/SubUV/SubUVAnimation(Number Of Frames=&FloatFromCurve(FloatCurve=DataInterface() {propertyValues {Curve {keys [
         {arriveTangent=136.8638, interpMode=RCIM_Cubic, time=0, value=10}
         {arriveTangent=113.791, interpMode=RCIM_Cubic, leaveTangent=113.791, tangentMode=RCTM_User, time=1, value=63}
        ]}}}))
   &ScaleColor(
    Scale Alpha=&FloatFromCurve(FloatCurve=DataInterface() {propertyValues {Curve {keys [{arriveTangent=-0.3, interpMode=RCIM_Cubic, leaveTangent=-0.3, tangentMode=RCTM_User, time=0.25, value=0.1}, <at 1 0 RCIM_Cubic>]}}})
    Scale RGB=&VectorFromCurve(VectorCurve=DataInterface() {propertyValues {
       XCurve {keys [<at 0 5 RCIM_Cubic>,<at 0.6 1 RCIM_Cubic>]}
       YCurve {keys [<at 0 2.7379 RCIM_Cubic>,<at 0.6 1 RCIM_Cubic>]}
       ZCurve {keys [<at 0 2.5 RCIM_Cubic>,<at 0.6 1 RCIM_Cubic>]}
      }})
    ScaleA=true
    ScaleRGB=true
   )
   &DynamicMaterialParameters(
    Index 0 Param 1=&FloatFromCurve(FloatCurve=DataInterface() {propertyValues {Curve {keys [<at 0 1 RCIM_Cubic>,<at 1 1 RCIM_Cubic>]}}})
    Index 0 Param 2=1
    Index 0 Param 3=1
    Index 0 Param 4=1
    Module.Param0WriteEnabled=true
    Module.Param1WriteEnabled=true
    Module.Param2WriteEnabled=true
    Module.Param3WriteEnabled=true
   )
   &CameraOffset(Camera Offset Amount=3)
   /SetVariables(Particles.CustomPos:NiagaraPosition=&AddVectorToPosition(
     Position=$Output.SocketLocation.SampledPosition
     Vector=&Subtract_Position(A=$Particles.Position, B=/Niagara/DynamicInputs/Helpers/SimulationPosition())
    )) {isSetParameters=true}
  ]
  rendererClasses [&NiagaraSpriteRendererProperties]
  renderers [{rendererClass=&NiagaraSpriteRendererProperties}]
 }
 StreakySmoke {
  emitterUpdateScript [
   &EmitterState(
    Delay First Loop Only=true
    Life Cycle Mode=::NewEnumerator1
    Loop Behavior=::NewEnumerator2
    Loop Count=7
    Loop Delay=&FN_FrontEndFloatModifier(Gameplay=1.65, Front End=1.65)
    Loop Duration=0.1
    UseLoopDelay=true
   )
   &SpawnBurst_Instantaneous(Spawn Count=10, Use Spawn Probability=false, UseLoopCountLimit=false)
  ]
  particleSpawnScript [
   &InitializeParticle(
    Color=&MultiplyLinearColor(
     Linear Color=<rgba 0.8755 0.9093 1 1.5>
     Scale Factor=&MakeVector4_FromLinearColor(LinearColor=$System.TOD)
    )
    Lifetime=&RandomRangeFloat(Minimum=0.2, Maximum=0.3)
    Sprite Size=&RandomRangeVector2D(Minimum=<30 120>, Maximum=<50 150>)
    Sprite Size Mode=::NewEnumerator4
    Sprite UV Mode=::NewEnumerator4
    Sprite UV Scale:Vector2f=&MakeCustomVector2DFromBool(Bool=&RandomBool(), True Vector2D=<-1 1>, False Vector2D=<-1 -1>)
    UsePositionOffset=false
    Write Material Random=true
   )
   &AddVelocity(
    Cone Angle=10
    Cone Axis=&Multiply_Vector(
     A=&MakeCustomVectorFromBool(
      Bool=&SetBoolByIntComparison(A=$Emitter.LoopCount, B=0, Comparison Type=::NewEnumerator4)
      True Vector=&RandomRangeVector(Minimum=<-0.3 0 -1>, Maximum=<0.3 -0.5 -0.5>)
      False Vector=&Multiply_Vector(
       A=&MakeCustomVectorFromBool(Bool=&RandomBool(), True Vector=<-1 1 1>, False Vector=<1 1 1>)
       B=&RandomRangeVector(Minimum=<0.2 -1 -0.3>, Maximum=<0.5 -1 -0.5>)
      )
     )
     B=&FN_FrontEndVector3Modifier(Gameplay=<1 1 1>, Front End=<-1 -1 0.3>)
    )
    Distribution Along Cone Axis:NiagaraFloat=&RandomRangeFloat(Minimum=0.02, Maximum=0.05)
    Module.UseConeFalloff=false
    Module.UseVelDistribution=true
    Override Seed=false
    Velocity Mode=::NewEnumerator2
    Velocity Speed=&RandomRangeFloat(Minimum=20, Maximum=50)
   )
   &CameraOffset(Camera Offset Amount=10)
   /SetVariables(Particles.CustomID:NiagaraInt32=&MakeCustomIntFromBool(
     Bool=&SetBoolByIntComparison(A=$Emitter.LoopCount, B=0, Comparison Type=::NewEnumerator4)
     True INT=0
     False INT=&Subtract_Integer(
      A:NiagaraInt32=&Multiply_Int(A=$Emitter.LoopCount, B=2)
      B:NiagaraInt32=&ModuloInt(A=&ReturnExecIndex(), B=2)
     )
    )) {isSetParameters=true}
  ]
  particleUpdateScript [
   &ParticleState()
   &SocketLocation(
    Interpolate Socket Sampling=true
    Orientation Sampling=::NewEnumerator1
    Position Sampling=::NewEnumerator1
    SamplePosOffset=true
    Sampled Position Offset=&MakeCustomVectorFromBool(
     Bool=&SetBoolByIntComparison(A=$Emitter.LoopCount, B=0, Comparison Type=::NewEnumerator4)
     True Vector=<0 0 0>
     False Vector=&RandomRangeVector(Minimum=<-10 0 0>, Maximum=<10 0 0>)
    )
    Sampled Position Offset Coordinate Space:ENiagaraSkelMeshTransforms=::NewEnumerator2
    Socket Index=$Particles.Initial.CustomID
    Socket Reader=$System.Socket Reader
    Socket Sampling Mode=::NewEnumerator3
    VelocitySpeedLimit=false
   )
   &ScaleSpriteSize(
    Non-Uniform Curve Sprite Scale=DataInterface() {propertyValues {
      XCurve {keys [
        {arriveTangent=2.3729, interpMode=RCIM_Cubic, leaveTangent=2.3729, leaveTangentWeight=0.3333, tangentMode=RCTM_User, time=0, value=0}
        {
         arriveTangent=1.5
         arriveTangentWeight=0.1667
         interpMode=RCIM_Cubic
         leaveTangent=1.5
         leaveTangentWeight=0.1667
         tangentMode=RCTM_User
         time=0.5
         value=0.8
        }
        {arriveTangent=9.2697, arriveTangentWeight=0.3333, interpMode=RCIM_Cubic, leaveTangent=9.2697, tangentMode=RCTM_User, time=1, value=3}
       ]}
      YCurve {keys [
        {arriveTangent=2.3729, interpMode=RCIM_Cubic, leaveTangent=2.3729, leaveTangentWeight=0.3333, tangentMode=RCTM_User, time=0, value=0}
        {
         arriveTangent=1.5
         arriveTangentWeight=0.1667
         interpMode=RCIM_Cubic
         leaveTangent=1.5
         leaveTangentWeight=0.1667
         tangentMode=RCTM_User
         time=0.5
         value=0.8
        }
        {arriveTangent=0.7753, arriveTangentWeight=0.3333, interpMode=RCIM_Cubic, leaveTangent=0.7753, tangentMode=RCTM_User, time=1, value=1}
       ]}
     }}
    Scale Sprite Size Mode=::NewEnumerator3
   )
   /Niagara/Modules/Update/SubUV/V2/SubUVAnimation(
    Random Start Frame=true
    "Sprite Renderer "=DataInterface() {propertyValues {SpriteRenderer=&self:StreakySmoke_1.NiagaraSpriteRendererProperties_1}}
    SubUV Lookup Index:NiagaraFloat=&FloatFromCurve(
     FloatCurve=DataInterface() {propertyValues {Curve {keys [
         {arriveTangent=136.8638, interpMode=RCIM_Cubic, leaveTangent=136.8638, leaveTangentWeight=0.3333, tangentMode=RCTM_User, time=0, value=16}
         {arriveTangentWeight=0.3333, interpMode=RCIM_Cubic, tangentMode=RCTM_User, time=1, value=63}
        ]}}}
     Scale Curve=&Multiply_Float(A=0.0159, B=0.1)
    )
    UseEndFrame=false
    UseStartFrame=false
   ) {module=SubUVAnimation001}
   &ScaleColor(
    Scale Alpha=&FloatFromCurve(FloatCurve=DataInterface() {propertyValues {Curve {keys [
         <at 0 0 RCIM_Cubic>
         {arriveTangent=0.2857, interpMode=RCIM_Cubic, leaveTangent=0.2857, time=0.15, value=0.1}
         {arriveTangent=-0.2311, interpMode=RCIM_Cubic, leaveTangent=-0.2311, tangentMode=RCTM_User, time=0.35, value=0.1}
         <at 1 0 RCIM_Cubic>
        ]}}})
    ScaleA=true
    ScaleRGB=true
   )
   &SolveForcesAndVelocity(Clamp Velocity=false, Limit Acceleration=false)
   /SetVariables(Particles.CustomPos:NiagaraPosition=&AddVectorToPosition(
     Position=$Output.SocketLocation.SampledPosition
     Vector=&Subtract_Position(A=$Particles.Position, B=&SimulationPosition())
    )) {isSetParameters=true}
  ]
  rendererClasses [&NiagaraSpriteRendererProperties]
  renderers [{rendererClass=&NiagaraSpriteRendererProperties}]
 }
 WingStreaks {
  emitterUpdateScript [
   &EmitterState(
    Delay First Loop Only=true
    Life Cycle Mode=::NewEnumerator1
    Loop Behavior=::NewEnumerator2
    Loop Delay=1
    Loop Duration=1
    UseLoopDelay=true
   )
   &SpawnBurst_Instantaneous(Spawn Count=10, Use Spawn Probability=false, UseLoopCountLimit=false)
  ]
  particleSpawnScript [
   &InitializeParticle(
    Color=&MultiplyLinearColor(
     Linear Color=<rgba 0.8755 0.9093 1 1.5>
     Scale Factor=&MakeVector4_FromLinearColor(LinearColor=$System.TOD)
    )
    Lifetime=&RandomRangeFloat(Minimum=0.2, Maximum=0.3)
    Sprite Size=&RandomRangeVector2D(Minimum=<80 300>, Maximum=<120 500>)
    Sprite Size Mode=::NewEnumerator4
    Sprite UV Mode=::NewEnumerator4
    Sprite UV Scale:Vector2f=&MakeCustomVector2DFromBool(Bool=&RandomBool(), True Vector2D=<-1 1>, False Vector2D=<-1 -1>)
    UsePositionOffset=false
    Write Material Random=true
   )
   &SocketLocation(
    Interpolate Socket Sampling=true
    Orientation Sampling=::NewEnumerator1
    SamplePosOffset=true
    Sampled Position Offset=&MakeCustomVectorFromBool(
     Bool=&SetBoolByIntComparison(A=$Emitter.LoopCount, B=0, Comparison Type=::NewEnumerator4)
     True Vector=<0 0 0>
     False Vector=&RandomRangeVector(Minimum=<-10 0 0>, Maximum=<10 0 0>)
    )
    Sampled Position Offset Coordinate Space:ENiagaraSkelMeshTransforms=::NewEnumerator2
    Socket Reader=$System.Socket Reader
    Socket Sampling Mode=::NewEnumerator3
    VelocitySpeedLimit=false
   )
   &AddVelocity(
    Cone Angle=10
    Cone Axis=&Add_Vector(
     A=&RandomRangeVector(Minimum=<0 0 0>, Maximum=<0 -0.3 0.5>)
     B=&MakeCustomVectorFromBool(Bool=&RandomBool(), True Vector=<-1 0 0>, False Vector=<1 0 0>)
    )
    Distribution Along Cone Axis:NiagaraFloat=&RandomRangeFloat(Minimum=0.02, Maximum=0.05)
    Module.UseConeFalloff=false
    Module.UseVelDistribution=true
    Override Seed=false
    Velocity Mode=::NewEnumerator2
    Velocity Speed=&RandomRangeFloat(Minimum=10, Maximum=30)
   )
   &CameraOffset(Camera Offset Amount=10)
  ]
  particleUpdateScript [
   &ParticleState()
   &ScaleSpriteSize(
    Non-Uniform Curve Sprite Scale=DataInterface() {propertyValues {
      XCurve {keys [
        {arriveTangent=2.3729, interpMode=RCIM_Cubic, leaveTangent=2.3729, leaveTangentWeight=0.3333, tangentMode=RCTM_User, time=0, value=0}
        {
         arriveTangent=1.5
         arriveTangentWeight=0.1667
         interpMode=RCIM_Cubic
         leaveTangent=1.5
         leaveTangentWeight=0.1667
         tangentMode=RCTM_User
         time=0.5
         value=0.8
        }
        {arriveTangent=9.2697, arriveTangentWeight=0.3333, interpMode=RCIM_Cubic, leaveTangent=9.2697, tangentMode=RCTM_User, time=1, value=3}
       ]}
      YCurve {keys [
        {arriveTangent=2.3729, interpMode=RCIM_Cubic, leaveTangent=2.3729, leaveTangentWeight=0.3333, tangentMode=RCTM_User, time=0, value=0}
        {
         arriveTangent=1.5
         arriveTangentWeight=0.1667
         interpMode=RCIM_Cubic
         leaveTangent=1.5
         leaveTangentWeight=0.1667
         tangentMode=RCTM_User
         time=0.5
         value=0.8
        }
        {arriveTangent=0.7753, arriveTangentWeight=0.3333, interpMode=RCIM_Cubic, leaveTangent=0.7753, tangentMode=RCTM_User, time=1, value=1}
       ]}
     }}
    Scale Sprite Size Mode=::NewEnumerator3
   )
   &SubUVAnimation(
    Random Start Frame=true
    "Sprite Renderer "=DataInterface() {propertyValues {SpriteRenderer=&self:WingStreaks_0.NiagaraSpriteRendererProperties_1}}
    SubUV Lookup Index:NiagaraFloat=&FloatFromCurve(
     FloatCurve=DataInterface() {propertyValues {Curve {keys [
         {arriveTangent=136.8638, interpMode=RCIM_Cubic, leaveTangent=136.8638, leaveTangentWeight=0.3333, tangentMode=RCTM_User, time=0, value=16}
         {arriveTangentWeight=0.3333, interpMode=RCIM_Cubic, tangentMode=RCTM_User, time=1, value=63}
        ]}}}
     Scale Curve=&Multiply_Float(A=0.0159, B=0.1)
    )
    UseEndFrame=false
    UseStartFrame=false
   ) {module=SubUVAnimation001}
   &ScaleColor(
    Scale Alpha=&FloatFromCurve(FloatCurve=DataInterface() {propertyValues {Curve {keys [
         <at 0 0 RCIM_Cubic>
         {arriveTangent=0.2857, interpMode=RCIM_Cubic, leaveTangent=0.2857, time=0.15, value=0.1}
         {arriveTangent=-0.2311, interpMode=RCIM_Cubic, leaveTangent=-0.2311, tangentMode=RCTM_User, time=0.35, value=0.1}
         <at 1 0 RCIM_Cubic>
        ]}}})
    ScaleA=true
    ScaleRGB=true
   )
   &SolveForcesAndVelocity(Clamp Velocity=false, Limit Acceleration=false)
  ]
  rendererClasses [&NiagaraSpriteRendererProperties]
  renderers [{rendererClass=&NiagaraSpriteRendererProperties}]
 }
}
summary {
 systemName=NS_Glider_DragonPry_Smoke
 emitters [
  {emitterName=WingStreaks, rendererClasses [&NiagaraSpriteRendererProperties]}
  {emitterName=BladeFlares, rendererClasses [&NiagaraSpriteRendererProperties]}
  {emitterName=DeployTrail, rendererClasses [&NiagaraRibbonRendererProperties]}
  {emitterName=Smoke, rendererClasses [&NiagaraSpriteRendererProperties]}
  {emitterName=StreakySmoke, rendererClasses [&NiagaraSpriteRendererProperties]}
 ]
 userVariables [
  User.bIsFrontEnd {type=NiagaraBool}
  User.bIsFrontEndPreview {type=NiagaraBool}
  User.RibbonIDOffset {type=NiagaraInt32}
  User.Variant {type=NiagaraFloat, defaultValue=1, description=""}
 ]
}
